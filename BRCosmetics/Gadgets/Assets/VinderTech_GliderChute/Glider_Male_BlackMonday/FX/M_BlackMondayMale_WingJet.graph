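# Additive, unlit, two-sided mesh-particle material for the wing jet effect on the Black Monday male glider.
# Two panning pattern and smoke textures multiply with particle color into emissive adjustments.
# Opacity blends complex or simplified procedural UV-gradient masks, shaped by depth fade, inverted Fresnel and particle alpha.
# World position offset follows a vector masked along V, with optional vertex-color symmetry and exclusion; dynamic parameters drive fade, pan speed and offset.
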Material /BRCosmetics/Gadgets/Assets/VinderTech_GliderChute/Glider_Male_BlackMonday/FX/M_BlackMondayMale_WingJet
graph {
 properties {
  allowTranslucentCustomDepthWrites=true
  bUsedWithMeshParticles=true
  bUsedWithNiagaraMeshParticles=true
  blendMode=BLEND_Additive
  pixelDepthOffsetMode=PDOM_Legacy
  shadingModel=MSM_Unlit
  twoSided=true
  usedShadingModels=Unlit
 }
 nodes [
  n1 = Multiply(
   A=TextureCoordinate() {vTiling=0.5}
   B=ComponentMask(None=VectorParameter() {parameterName=Texture A UV Scale, defaultValue=<rgba 1 1 0 0>}.RGB) {g=true, r=true}
  )
  n2 = Multiply(
   A=Time()
   B=Multiply(A=DynamicParameter() {defaultValue=<rgba 100 2 0 1>, paramNames [DepthFadeDistance, Panning Speed, WPO Amount, Param4]}.Panning Speed) {constB=-1}
  )
  n3 = Multiply(
   A=TextureCoordinate() {vTiling=0.5}
   B=ComponentMask(None=VectorParameter() {parameterName=Texture B UV Scale, defaultValue=<rgba 1 1 0 0>}.RGB) {g=true, r=true}
  )
  n4 = Multiply(
   A=TextureSample(UVs=Panner(Coordinate=@n1, Time=@n2) {speedY=0.25}) {texture=/Game/Effects/Fort_Effects/Textures/T_Tile_Pattern_01}.R
   B=TextureSample(UVs=Panner(Coordinate=@n3, Time=@n2) {speedY=1}) {texture=/Game/Global/Textures/Default/Tileable_Patterns/T_Smoke_Puff_Tile_Up}.R
  )
  n5 = ParticleColor()
  MaterialFunctionCall(Emissive=Saturate(None=Multiply(A=@n4, B=@n5.RGB))) {materialFunction=&MF_FX_EmissiveAdjustments}
  n6 = VectorParameter() {group=WPO, parameterName=WPO_Vector, sortPriority=2, defaultValue=<rgba 1 0 0 1>}
  n7 = LinearInterpolate(A=@n6.RGB, B=Multiply(A=@n6.RGB, B=Multiply(A=@n6.RGB) {constB=-1}), Alpha=VertexColor().R)
  n8 = ScalarParameter() {group=WPO, parameterName=Use Symmetical WPO, sortPriority=1}
  n9 = ComponentMask(None=TextureCoordinate()) {g=true}
  n10 = ScalarParameter() {defaultValue=10, group=WPO, parameterName=WPO Contrast}
  n11 = Saturate(None=Multiply(A=Power(Base=@n9, Exp=@n10)) {constB=10})
  n12 = OneMinus(None=VertexColor().B)
  Multiply(
   A=Multiply(
    A=LinearInterpolate(A=@n6.RGB, B=@n7, Alpha=@n8)
    B=DynamicParameter() {defaultValue=<rgba 100 2 0 1>, paramNames [DepthFadeDistance, Panning Speed, WPO Amount, Param4]}.WPO Amount
   )
   B=LinearInterpolate(A=@n11, B=Multiply(A=@n11, B=@n12), Alpha=ScalarParameter() {group=WPO, parameterName=Remove Blue Vertex from WPO, sortPriority=3})
  )
  n13 = OneMinus(None=ComponentMask(None=TextureCoordinate()) {g=true})
  n14 = ScalarParameter() {defaultValue=1.75, group=Complex UV Alpha Mask, parameterName=Rear Opacity Contrast}
  n15 = Saturate(None=Multiply(A=Power(Base=@n13, Exp=@n14)) {constB=10})
  n16 = OneMinus(None=ComponentMask(None=TextureCoordinate()) {r=true})
  n17 = TextureCoordinate()
  n18 = Power(
   Base=OneMinus(None=ComponentMask(None=@n17) {g=true})
   Exp=ScalarParameter() {defaultValue=6.2611, group=Complex UV Alpha Mask, parameterName=Inside Contrast Selection}
  )
  n19 = LinearInterpolate(
   A=ScalarParameter() {defaultValue=3.8879, group=Complex UV Alpha Mask, parameterName=Central Opacity Contrast A}
   B=ScalarParameter() {defaultValue=5.5, group=Complex UV Alpha Mask, parameterName=Central Opacity Contrast B}
   Alpha=Saturate(None=Multiply(A=@n18) {constB=8})
  )
  n20 = Saturate(None=Multiply(A=Power(Base=@n16, Exp=@n19)) {constB=10})
  n21 = TextureCoordinate()
  n22 = ScalarParameter() {defaultValue=0.916, group=Complex UV Alpha Mask, parameterName=Offside Contrast A}
  n23 = ScalarParameter() {defaultValue=4, group=Complex UV Alpha Mask, parameterName=Offside Contrast B}
  n24 = ComponentMask(None=TextureCoordinate()) {g=true}
  n25 = ScalarParameter() {defaultValue=10, group=Complex UV Alpha Mask, parameterName=Annoying Corner Contrast}
  n26 = Saturate(None=Multiply(A=Power(Base=@n24, Exp=@n25)) {constB=8})
  n27 = Multiply(A=Power(Base=ComponentMask(None=@n21) {r=true}, Exp=LinearInterpolate(A=@n22, B=@n23, Alpha=@n26))) {constB=10}
  n28 = ComponentMask(None=TextureCoordinate()) {r=true}
  n29 = ScalarParameter() {defaultValue=2, group=Simplified UV Alpha Mask, parameterName=U1 Contrast, sortPriority=2}
  n30 = OneMinus(None=@n28)
  n31 = ScalarParameter() {defaultValue=2, group=Simplified UV Alpha Mask, parameterName=U0 Contrast, sortPriority=1}
  n32 = Multiply(A=Multiply(A=Power(Base=@n28, Exp=@n29), B=Power(Base=@n30, Exp=@n31))) {constB=10}
  n33 = ComponentMask(None=TextureCoordinate()) {g=true}
  n34 = ScalarParameter() {group=Simplified UV Alpha Mask, parameterName=V1 Contrast, sortPriority=4}
  n35 = OneMinus(None=@n33)
  n36 = ScalarParameter() {defaultValue=3.1098, group=Simplified UV Alpha Mask, parameterName=V0 Contrast, sortPriority=3}
  n37 = Multiply(A=Multiply(A=Power(Base=@n33, Exp=@n34), B=Power(Base=@n35, Exp=@n36))) {constB=10}
  n38 = LinearInterpolate(
   A=Multiply(A=Multiply(A=@n15, B=@n20), B=Saturate(None=@n27))
   B=Multiply(A=Saturate(None=@n32), B=Saturate(None=@n37))
   Alpha=ScalarParameter() {group=Simplified UV Alpha Mask, parameterName=Use Simplified UV Alpha Mask}
  )
  n39 = DynamicParameter() {defaultValue=<rgba 100 2 0 1>, paramNames [DepthFadeDistance, Panning Speed, WPO Amount, Param4]}
  n40 = OneMinus(None=Fresnel(ExponentIn=ScalarParameter() {defaultValue=0.1, parameterName=FresnelExp}))
  Multiply(A=Multiply(A=@n5.A, B=DepthFade(Opacity=@n38, FadeDistance=@n39.DepthFadeDistance)), B=Saturate(None=Multiply(A=@n40) {constB=10}))
 ]
}
