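# Niagara ribbon trail effect for the BlackMonday male glider, named for its deploy animation.
# A single Trail emitter continuously spawns short-lived particles through SkeletalMeshLocation, whose mesh interface previews the glider and filters to two FX_DeployTrail sockets.
# Particles start with a near-white color that ScaleColor darkens heavily in RGB during update.
# A separate SocketLocation module is present but disabled.
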NiagaraSystem /BRCosmetics/Gadgets/Assets/VinderTech_GliderChute/Glider_Male_BlackMonday/FX/NS_Glider_BlackMondayMale_DeployAnimTrail_Converted
dependencies {
 usedDataInterfaces [&NiagaraDataInterfaceSocketReader, &NiagaraDataInterfaceSkeletalMesh]
 usedDynamicInputs []
 usedModules [&SampleSkeletalMesh, &SystemState]
}
emitters {Trail {
  emitterUpdateScript [
   &EmitterState(Life Cycle Mode=::NewEnumerator1, Loop Delay=0.025, Loop Duration=1, UseLoopDelay=true)
   &SpawnRate(SpawnRate=40, Use Spawn Probability=false)
  ]
  particleSpawnScript [
   &InitializeParticle(Lifetime=0.3, UsePositionOffset=false, Write Material Random=true)
   &Color(Color=&MakeLinearColorFromVectorAndFloat("Vector (RGB)"=<0.9393 1 0.9687>, "Float (Alpha)"=1.5))
   &SocketLocation(OverrideRandomSeed=false, SamplePosOffset=false, VelocitySpeedLimit=false) {enabled=false}
   &SkeletalMeshLocation(
    MaskSpawnGroup=false
    OverrideRandomSeed=false
    Skeletal Mesh=DataInterface() {propertyValues {
      DefaultMesh=null
      ExcludeBoneName=root
      OptionalComponentClass=null
      PreviewMesh=/BRCosmetics/Gadgets/Assets/VinderTech_GliderChute/Glider_Male_BlackMonday/Meshes/BlackMonday_Male_Glider
      SkinningMode=SkinOnTheFly
      SoftSourceActor=null
      SourceMode=Default
      UvSetIndex=0
      WholeMeshLOD=-1
      bExcludeBone=false
      bReadDeformedGeometry=true
      bRequireCurrentFrameData=true
      ComponentTags []
      FilteredBones []
      FilteredSockets [FX_DeployTrail_01, FX_DeployTrail_02]
      MeshUserParameter {parameter {name=None, typeDefHandle {registeredTypeIndex=1}, varData []}}
      SamplingRegions []
     }}
   )
  ]
  particleUpdateScript [
   &ParticleState()
   &ScaleColor(Scale RGB=<0.0312 0.0312 0.0312>, ScaleA=true, ScaleRGB=true)
  ]
  rendererClasses [&NiagaraRibbonRendererProperties]
  renderers [{rendererClass=&NiagaraRibbonRendererProperties}]
 }}
summary {
 systemName=NS_Glider_BlackMondayMale_DeployAnimTrail_Converted
 emitters [{emitterName=Trail, rendererClasses [&NiagaraRibbonRendererProperties]}]
 userVariables []
}
