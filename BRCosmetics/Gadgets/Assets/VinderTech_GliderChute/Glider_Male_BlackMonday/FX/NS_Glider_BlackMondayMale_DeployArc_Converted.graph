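# Niagara effect for the Black Monday male glider's deploy arc, a short-lived mesh flash.
# A single emitter burst-spawns mesh particles with a brightened color and randomized material values.
# A curve on particle alpha fades it quickly in and slowly out over each particle's life.
# Dynamic material parameters, two of them curve-animated, drive the mesh material over that life.
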NiagaraSystem /BRCosmetics/Gadgets/Assets/VinderTech_GliderChute/Glider_Male_BlackMonday/FX/NS_Glider_BlackMondayMale_DeployArc_Converted
dependencies {usedDataInterfaces [&NiagaraDataInterfaceCurve], usedDynamicInputs [], usedModules [&SystemState]}
emitters {Particle_Emitter {
  emitterUpdateScript [
   &EmitterState(Life Cycle Mode=::NewEnumerator1, Loop Behavior=::NewEnumerator1, Loop Duration=1, UseLoopDelay=true)
   &SpawnBurst_Instantaneous(Use Spawn Probability=false, UseLoopCountLimit=false)
  ]
  particleSpawnScript [
   &InitializeParticle(Lifetime=1.2, Mesh Scale Mode=::NewEnumerator4, UsePositionOffset=false, Write Material Random=true)
   &Color(Color=&MakeLinearColorFromVectorAndFloat("Vector (RGB)"=<2 2 2>, "Float (Alpha)"=2))
  ]
  particleUpdateScript [
   &ParticleState()
   &ScaleColor(
    Scale Alpha=&FloatFromCurve(FloatCurve=DataInterface() {propertyValues {Curve {keys [
         {arriveTangent=9.2959, interpMode=RCIM_Cubic, leaveTangent=9.2959, tangentMode=RCTM_User, time=0, value=0}
         <at 0.2 1 RCIM_Cubic>
         <at 1 0 RCIM_Cubic>
        ]}}})
    ScaleA=true
    ScaleRGB=true
   )
   &DynamicMaterialParameters(
    Index 0 Param 1=&FloatFromCurve(FloatCurve=DataInterface() {propertyValues {Curve {keys [
         {arriveTangent=987.5064, interpMode=RCIM_Cubic, leaveTangent=987.5064, tangentMode=RCTM_User, time=0, value=0}
         {interpMode=RCIM_Cubic, tangentMode=RCTM_User, time=0.25, value=100}
         <at 1 250 RCIM_Cubic>
        ]}}})
    Index 0 Param 2=10
    Index 0 Param 3=&FloatFromCurve(FloatCurve=DataInterface() {propertyValues {Curve {keys [<at 0 50>,<at 0.5 0>]}}})
    Index 0 Param 4=1
    Module.Param0WriteEnabled=true
    Module.Param1WriteEnabled=true
    Module.Param2WriteEnabled=true
    Module.Param3WriteEnabled=true
   )
  ]
  rendererClasses [&NiagaraMeshRendererProperties]
  renderers [{rendererClass=&NiagaraMeshRendererProperties, rendererIndex=0}]
 }}
summary {
 systemName=NS_Glider_BlackMondayMale_DeployArc_Converted
 emitters [{emitterName=Particle_Emitter, rendererClasses [&NiagaraMeshRendererProperties]}]
 userVariables []
}
